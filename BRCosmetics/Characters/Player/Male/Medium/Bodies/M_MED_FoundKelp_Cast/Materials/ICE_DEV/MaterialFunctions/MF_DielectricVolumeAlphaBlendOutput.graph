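# Material function that composites a dielectric volume, described as water, into luminance and opacity for alpha-blended output.
# It combines surface diffuse, specular and opacity mask with volume luminance, transmittance and scene luminance.
# The minimum transmittance channel sets opacity behind the surface, while the colored excess, with optional haze, tints scene luminance folded into the output.
# Final luminance is un-premultiplied by the blended opacity, clamped against zero division.
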MaterialFunction /BRCosmetics/Characters/Player/Male/Medium/Bodies/M_MED_FoundKelp_Cast/Materials/ICE_DEV/MaterialFunctions/MF_DielectricVolumeAlphaBlendOutput
graph {nodes [
  n1 = LinearInterpolate(A=NamedRerouteUsage(), Alpha=NamedRerouteUsage())
  FunctionOutput(None=@n1) {outputName="Opacity (S)", sortPriority=10}
  n2 = NamedRerouteUsage()
  n3 = LinearInterpolate(A=NamedRerouteUsage(), B=NamedRerouteUsage(), Alpha=NamedRerouteUsage())
  FunctionOutput(None=Divide(A=Add(A=@n2, B=@n3), B=Max(A=@n1) {constB=0.0001}) {desc=unpremult}) {outputName="Luminance (V3)"}
  NamedRerouteDeclaration(None=FunctionInput() {bUsePreviewValueAsDefault=true, inputName=Diffuse Luminance, sortPriority=120}) {name="Diffuse Luminance (V3)", nodeColor=<rgba 1 0.8333 0 1>}
  NamedRerouteDeclaration(None=FunctionInput() {bUsePreviewValueAsDefault=true, inputName=Volume Luminance, sortPriority=200}) {name="Volume Luminance (V3)", nodeColor=<rgba 0 1 0.1094 1>}
  n4 = Multiply(A=NamedRerouteUsage(), B=LinearInterpolate(A=NamedRerouteUsage(), Alpha=NamedRerouteUsage()))
  n5 = Multiply(A=NamedRerouteUsage(), B=@n4) {desc=Haze Color}
  n6 = Min(A=ComponentMask(None=@n4) {r=true}, B=Min(A=ComponentMask(None=@n4) {g=true}, B=ComponentMask(None=@n4) {b=true})) {desc=Min Transmittance}
  n7 = Subtract(A=@n4, B=@n6)
  n8 = Multiply(A=NamedRerouteUsage(), B=Add(A=Multiply(A=@n5, B=@n7), B=@n7))
  n9 = NamedRerouteUsage()
  NamedRerouteDeclaration(None=Multiply(A=Add(A=@n8, B=@n9), B=NamedRerouteUsage())) {name="Luminance Behind Surface (V3)", nodeColor=<rgba 0.1953 0 1 1>}
  n10 = FunctionInput() {bUsePreviewValueAsDefault=true, inputName=Volume Transmittance, sortPriority=210, previewValue=<1 1 1 1>}
  NamedRerouteDeclaration(None=@n10) {name="Volume Transmittance (V3)", nodeColor=<rgba 0 1 0.6016 1>}
  n11 = FunctionInput() {bUsePreviewValueAsDefault=true, inputName=Surface Transmittance, inputType=FunctionInput_Scalar, sortPriority=150, previewValue=<1 0 0 1>}
  NamedRerouteDeclaration(None=@n11) {name="Surface Transmittance (S)", nodeColor=<rgba 1 0 0.7031 1>}
  NamedRerouteDeclaration(None=FunctionInput() {bUsePreviewValueAsDefault=true, inputName=Specular Luminance, sortPriority=110}) {name="Specular Luminance (V3)", nodeColor=<rgba 0 0.4609 1 1>}
  NamedRerouteDeclaration(None=FunctionInput() {bUsePreviewValueAsDefault=true, inputName=Scene Luminance, sortPriority=250}) {name="Scene Luminance (V3)", nodeColor=<rgba 0.5 0 1 1>}
  n12 = FunctionInput() {
   bUsePreviewValueAsDefault=true
   description="Diffuse opacity of the water surface; 0 : full visibility of the underwater scene + water volume scattering; 1 : the water surface is a fully opaque lambertian reflector; does not affect surface specular/reflection (always visible)"
   inputName=Surface Opacity Mask
   inputType=FunctionInput_Scalar
   sortPriority=140
  }
  NamedRerouteDeclaration(None=@n12) {name="Surface Opacity Mask (S)", nodeColor=<rgba 0.4297 0 1 1>}
  NamedRerouteDeclaration(None=FunctionInput() {bUsePreviewValueAsDefault=true, inputName=Is Scene Inside Volume, inputType=FunctionInput_Scalar, sortPriority=270}) {name="Is Scene Inside Volume (S)", nodeColor=<rgba 0.313 0 1 1>}
  NamedRerouteDeclaration(None=Saturate(None=OneMinus(None=@n6))) {name="Opacity Behind Surface (S)", nodeColor=<rgba 0 0.6957 1 1>}
  NamedRerouteDeclaration(None=FunctionInput() {bUsePreviewValueAsDefault=true, inputName=Haze Amount, inputType=FunctionInput_Scalar, sortPriority=260}) {name="Haze Amount  (S)", nodeColor=<rgba 0.3359 0 1 1>}
 ]}
